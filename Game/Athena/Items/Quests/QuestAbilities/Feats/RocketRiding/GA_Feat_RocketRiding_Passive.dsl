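; Passive gameplay ability for the rocket riding feat that tracks when the player stands on a projectile.
; On activation it binds to the Athena player pawn's OnPawnLanded event and inspects the hit actor of each landing.
; Landing on a FortProjectileBase applies the rocket riding gameplay effect to the owner, while any other landing removes it by asset tag.
; Ending the ability removes the effect and unbinds the landing event.
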(event Ability|EventActivateAbility
  (|CallParentFunction)
  (bind _asfort_player_pawn_athena (Utilities|Casting|CastToFortPlayerPawnAthena (Ability|GetAvatarActorfromActorInfo)))
  (EventDispatchers|BindEventtoOnPawnLanded _asfort_player_pawn_athena (EventDispatchers|CreateEvent :!event Custom|OnLand)))

(event Ability|EventFailedToActivatePassiveAbility
  (params (ActorInfo :type Gameplay%20Ability%20Actor%20Info%20Structure)))

(event Custom|OnLand
  (params (Hit :type Hit%20Result%20Structure%20%28by%20ref%29))
  (bind (_bblockinghit _binitialoverlap _time _distance _location _impactpoint _normal _impactnormal _physmat _hitactor) (Collision|BreakHitResult Hit))
  (Utilities|IsValid _hitactor
    (:Is%20Valid
      (Utilities|Casting|CastToFortProjectileBase _hitactor
        (:then
          (Ability|ApplyGameplayEffectToOwner (Variables|Default|GetRocketRidingGE)))
        (:CastFailed
          _)))
    (:Is%20Not%20Valid
      _))
  (Ability|RemoveGameplayEffectFromOwnerWithAssetTags (Variables|Default|GetRocketRidingGETag)))

(event Ability|EventOnEndAbility
  (params (bWasCancelled :type Boolean))
  (|CallParentFunction bWasCancelled)
  (CallFunction|AbilityCleanup))

(event Custom|AbilityCleanup
  (Ability|RemoveGameplayEffectFromOwnerWithAssetTags (Variables|Default|GetRocketRidingGETag))
  (bind _asfort_player_pawn_athena (Utilities|Casting|CastToFortPlayerPawnAthena (Ability|GetAvatarActorfromActorInfo)))
  (EventDispatchers|UnbindEventfromOnPawnLanded _asfort_player_pawn_athena (EventDispatchers|CreateEvent :!event Custom|OnLand)))
